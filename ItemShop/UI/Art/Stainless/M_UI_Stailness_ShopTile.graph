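# UI-domain material for a Stainless-themed item shop tile, drawing a gold-to-brown gradient with a rounded outline.
# A rotated linear gradient, eased with a sine curve, blends the two colors, with a faint tan overlay driven by a rerouted mask.
# The CoreUI rounded mask takes its corner settings from MPC_ItemShop and a Border Thickness parameter, and its Outline SDF drives the final color blend.
# A procedural Stripes pattern is built from sine waves of tiled UVs.
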Material /ItemShop/UI/Art/Stainless/M_UI_Stailness_ShopTile
graph {
 properties {
  bAutomaticallySetUsageInEditor=false
  bUsedWithStaticMesh=false
  materialDomain=MD_UI
  pixelDepthOffsetMode=PDOM_Legacy
  usedShadingModels=Default Lit
 }
 nodes [
  n1 = Constant3Vector() {constant=<rgba 0.3073 0.1382 0 1>}
  n2 = Constant3Vector() {constant=<rgba 1 0.8792 0 1>}
  n3 = MaterialFunctionCall("Value (0-1)"=MaterialFunctionCall(UVs=NamedRerouteUsage(), GradientStart=Constant(), GradientEnd=Constant() {r=1}, RotationAngle=Constant() {r=0.25}) {materialFunction=&MF_UI_LinearGradient}.Result) {materialFunction=&MF_UI_EaseCurves}
  n4 = NamedRerouteUsage()
  n5 = MaterialFunctionCall(
   Corner Radius=CollectionParameter() {collection=/FortUILibrary/Systems/ItemShopTile/MPC_ItemShop, parameterName=CornerRadius}
   Radius in Pixels=CollectionParameter() {collection=&MPC_ItemShop, parameterName=bRadiusInPixels}
   Outline Thickness=ScalarParameter() {defaultValue=1, parameterName=Border Thickness}
  ) {materialFunction=&MF_UI_CoreUI_RoundedMask}
  LinearInterpolate(
   A=@n1
   B=LinearInterpolate(
    A=LinearInterpolate(A=@n2, B=@n1, Alpha=@n3.EaseInOutSine)
    B=Constant3Vector() {constant=<rgba 0.3281 0.1966 0.0632 1>}
    Alpha=Multiply(A=@n4) {constB=0.2}
   )
   Alpha=@n5.Outline SDF
  )
  Multiply(A=SmoothStep(Value=@n5.Outline SDF) {constMax=0.1})
  NamedRerouteDeclaration(None=MaterialFunctionCall() {materialFunction=&MF_UI_UVScaling}.Scaled to Fit) {name=UVScaling, nodeColor=<rgba 0 1 0.8125 1>}
  n6 = NamedRerouteUsage()
  n7 = Frac(None=Multiply(A=ComponentMask(None=@n6) {r=true}) {constB=40})
  n8 = Frac(None=Multiply(A=ComponentMask(None=@n6) {g=true}) {constB=60})
  NamedRerouteDeclaration(None=Max(A=Sine(None=@n7), B=Sine(None=@n8))) {name=Stripes, nodeColor=<rgba 1 0 0.0703 1>}
  OneMinus(None=MaterialFunctionCall(
    UVs=NamedRerouteUsage()
    Size=Constant() {r=1}
    CornerRoundness=Constant() {r=0.3}
    OutlineThickness=Constant() {r=0.01}
    GlowMin=Constant()
    GlowMax=Constant() {r=0.1}
   ) {materialFunction=&MF_UI_Box}.Outline)
  StaticBool() {value=true}
  TextureCoordinate()
 ]
}
